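# Player animation Blueprint for the mounted cannon gunner, implementing vehicle override animation layers.
# The final-pose layer replaces its input with a cannon idle through a FullBody slot, plus a gender-selected additive with a female adjustment.
# The event graph holds an unwired chain computing offset world hand transforms from turret steering-wheel sockets, which no pose graph reads.
# The main graph and other override layers are empty.
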AnimBlueprint /Game/Animation/Game/MainPlayer/Combat/Gadgets/Medium/Male/MountedCannon/MountedCannon_Gunner_AnimBP
target_skeleton=&Fortnite_M_Avg_Player_Skeleton
graphs [
 AnimGraph {nodes [Output Pose() {layerGroup=None}]}
 EventGraph {nodes [
   n1 = GetSocketTransform(
    self=Get SkeletalMesh!(self:Fort Athena Vehicle Object Reference=Get MountedTurret() -> (MountedTurret:Fort Mounted Turret Object Reference)) -> (SkeletalMesh:Skeletal Mesh Component Budgeted Object Reference)
    InSocketName=steering_wheel_hand_l
    TransformSpace=RTS_World
   )
   Set WorldPosLeft(WorldPosLeft:Vector=TransformLocation!(T:"Transform (by ref)"=@n1, Location:Vector=Get LocalLeftOffset() -> (LocalLeftOffset:Vector)) -> (ReturnValue:Vector)) -> (Output_Get:Vector)
   > Set WorldRotLeft(WorldRotLeft:Rotator=TransformRotation!(T:"Transform (by ref)"=@n1, Rotation:Rotator=Get LocalRotLeftOffset() -> (LocalRotLeftOffset:Rotator)) -> (ReturnValue:Rotator)) -> (Output_Get:Rotator)
   n2 = GetSocketTransform(self=Get SkeletalMesh(self=Get MountedTurret()), InSocketName=steering_wheel_hand_r, TransformSpace=RTS_World)
   > Set WorldPosRight(WorldPosRight:Vector=TransformLocation(T=@n2, Location=Get LocalRightOffset() -> (LocalRightOffset:Vector))) -> (Output_Get:Vector)
   > Set WorldRotRight(WorldRotRight:Rotator=TransformRotation(T=@n2, Rotation=Get LocalRotRightOffset() -> (LocalRotRightOffset:Rotator))) -> (Output_Get:Rotator)
  ]}
 VehicleFinalPoseOverride {nodes [
   Input Pose() {name=InFinalPose}
   n1 = "Blend Poses (EFortCustomGender)"!(
    BlendPose_0:Pose Link Structure
    BlendPose_1:Pose Link Structure="Sequence Player (additive) 'MountedCannon_FemaleAdd_CMM'"() {sequence=&/MountedCannon_FemaleAdd_CMM}
    BlendTime_0:float=0.1
    BlendTime_1:float=0.1
    ActiveEnumValue:Byte=Get Gender() -> (Gender:EFortCustomGender Enum=Invalid)
   ) -> (Pose:Pose Link Structure) {
    is=title AnimGraphNode_BlendListByEnum
    blendType=Linear
    blendPose [{linkId=-1, sourceLinkId=-1}, {linkId=-1, sourceLinkId=-1}]
    blendTime [0.1,0.1]
   }
   n2 = Sequence Player 'MountedCannon_Idle_CMM'() {sequence=&/MountedCannon_Idle_CMM}
   Output Pose(Result=Apply Additive(Base=Slot 'FullBody'(Source=@n2) {slotName=FullBody}, Additive=@n1) {alphaBoolBlend {blendOption=Linear}}) {layerGroup=VehicleOverride}
  ]}
 VehicleFullBodyOverride {nodes [
   Input Pose() {name=InPoseFullBody}
   Output Pose() {layerGroup=VehicleOverride}
  ]}
 VehicleLowerBodyOverride {nodes [
   Input Pose() {name=InPoseLowerBodyDefault}
   Input Pose() {name=InPoseUpperBody}
   Output Pose() {layerGroup=VehicleOverride}
  ]}
 VehicleSplitBodyOverride {nodes [
   Input Pose() {name=InPoseSplitBody}
   Input Pose() {name=InPoseUpperAndLowerBody}
   Input Pose() {name=InPoseUpperAndLowerBodyRemoveRoot}
   Output Pose() {layerGroup=Vehicle Override}
  ]}
]
variables [
 isTurning {type=boolean}
 LocalLeftOffset {type=Vector, default=<-3.5 -0.5 -5>}
 LocalRightOffset {type=Vector, default=<3.5 -0.5 5>}
 LocalRotLeftOffset {type=Rotator}
 LocalRotRightOffset {type=Rotator}
 WorldPosLeft {type=Vector}
 WorldPosRight {type=Vector}
 WorldRotLeft {type=Rotator}
 WorldRotRight {type=Rotator}
]
